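# Translucent lit material for drifting snow front-end effects on instanced and Niagara meshes.
# World position offset combines camera-scaled spline thickening, vector-noise XY curl and looping per-instance Z travel, with optional camera offset.
# Opacity multiplies UV edge gradients, panning smoke and distorted turbulence, then applies near-camera, instance and depth fades.
# Color graphs blend time-of-day tints by sun-facing angle; curl and distortion drop out on cheaper quality or shading paths.
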Material /Game/Effects/Fort_Effects/FrontEnd/S4000/M_FX_FESnowDrift
graph {
 properties {
  bCastRayTracedShadows=false
  bUsedWithInstancedStaticMeshes=true
  bUsedWithNiagaraMeshParticles=true
  blendMode=BLEND_Translucent
  pixelDepthOffsetMode=PDOM_Legacy
  usedShadingModels=Default Lit
 }
 nodes [
  n1 = Constant() {r=1}
  n2 = ShadingPathSwitch(Default=ShadingPathSwitch(Default=Constant() {r=1}, Mobile=@n1) {desc=Migrated from FeatureLevelSwitch}, Forward=@n1, Mobile=@n1)
  n3 = VectorParameter() {parameterName=SnowColor, defaultValue=<rgba 1 1 1 1>}
  LinearInterpolate(
   A=LinearInterpolate(
    A=VectorParameter() {parameterName=BaseColor, defaultValue=<rgba 1 0.6843 0.1036 1>}.RGB
    B=QualitySwitch(Default=@n2, Low=@n1)
    Alpha=ScalarParameter() {parameterName=AmountOfRVTTexutreColor}
   )
   B=@n3.RGB
   Alpha=Constant()
  )
  n4 = CollectionParameter() {
   collection=&FortniteMaterialParameters
   parameterName=SunAndMoonModelDirectionalVector
  }
  n5 = Power(Base=DotProduct(A=ComponentMask(None=@n4) {b=true, g=true, r=true}, B=ReflectionVectorWS())) {constExponent=4}
  n6 = VectorParameter() {group=TOD, parameterName="Color_Morning (V3)", defaultValue=<rgba 1 1 1 0>}
  n7 = VectorParameter() {group=TOD, parameterName="Color_Day (V3)", defaultValue=<rgba 1 1 1 0>}
  n8 = VectorParameter() {group=TOD, parameterName="Color_Evening (V3)", defaultValue=<rgba 1 1 1 0>}
  n9 = VectorParameter() {group=TOD, parameterName="Color_Night (V3)", defaultValue=<rgba 1 1 1 0>}
  n10 = Multiply(
   A=LinearInterpolate(
    A=ScalarParameter() {defaultValue=0.5, group=Texture, parameterName=ColorIntensity_SunSide_Mult}
    B=ScalarParameter() {defaultValue=0.35, group=Texture, parameterName=ColorIntensity_ShadowSide_Mult}
    Alpha=Clamp(None=@n5)
   )
   B=VertexInterpolator(VS=MaterialFunctionCall(Color_Morning=@n6.RGB, Color_Day=@n7.RGB, Color_Evening=@n8.RGB, Color_Night=@n9.RGB) {materialFunction=&Color_Time_of_Day}.Result).PS
  )
  Multiply(A=@n10, B=@n3.RGB)
  Multiply(A=@n10, B=ScalarParameter() {parameterName=Emmisive})
  n11 = Saturate(None=MaterialFunctionCall(Distance=ScalarParameter() {defaultValue=100, parameterName=CameraScaleDistance}) {materialFunction=&Near_Camera_Fade}.Result)
  n12 = Clamp(None=@n11) {minDefault=0.1}
  n13 = Multiply(A=ScalarParameter() {defaultValue=10, parameterName="WidthBase (S)"}, B=@n12)
  Multiply(A=@n13)
  Multiply(A=ParticleColor().RGB, B=VectorParameter() {parameterName=MainColor, defaultValue=<rgba 1 1 1 1>}.RGB)
  NamedRerouteDeclaration(None=@n11) {name=NearFading, nodeColor=<rgba 0.2656 0 1 1>}
  n14 = ScalarParameter() {defaultValue=10, parameterName="WidthTip (S)"}
  n15 = ScalarParameter() {defaultValue=1, group=Movement, parameterName=Curl_Amount}
  n16 = WorldPosition()
  n17 = ObjectPositionWS()
  n18 = DynamicParameter() {paramNames [CurlRandom, CurlSpeed, OveralSpeedLoop, Param4]}
  n19 = ScalarParameter() {defaultValue=1, group=Movement, parameterName=Curl_Speed}
  n20 = Add(A=@n19, B=@n18.CurlSpeed)
  n21 = Multiply(A=@n19) {constA=0.65, constB=0.65}
  n22 = PerInstanceRandom()
  n23 = Subtract(A=Add(A=Subtract(A=@n16.XYZ, B=@n17), B=@n18.CurlRandom), B=Multiply(A=LinearInterpolate(A=@n20, B=@n21, Alpha=@n22), B=Time()))
  n24 = ScalarParameter() {defaultValue=1, group=Movement, parameterName=Curl_Scale}
  n25 = LinearInterpolate(A=@n24, B=Multiply(A=@n24) {constA=0.65, constB=0.65}, Alpha=PerInstanceRandom())
  n26 = Multiply(
   A=LinearInterpolate(A=@n15, B=Multiply(A=@n15) {constA=0.65, constB=0.65}, Alpha=PerInstanceRandom())
   B=VectorNoise(World Position=Divide(A=@n23, B=@n25)) {noiseFunction=VNF_VectorALU}
  )
  n27 = Constant()
  n28 = ShadingPathSwitch(
   Default=QualitySwitch(Default=ShadingPathSwitch(Default=@n26, Mobile=@n27) {desc=Migrated from FeatureLevelSwitch}, Low=@n27, Medium=@n27)
   Forward=@n27
  )
  n29 = ComponentMask(None=@n28) {r=true}
  n30 = ComponentMask(None=@n28) {g=true}
  n31 = ScalarParameter() {defaultValue=1, parameterName=ObjectRadiusMult}
  n32 = Multiply(A=@n31) {constB=-1}
  n33 = Divide(
   A=Time()
   B=Add(
    A=ScalarParameter() {defaultValue=6, group=1.Main values, parameterName=LoopDuration, sliderMax=50, sliderMin=0.1}
    B=DynamicParameter() {paramNames [CurlRandom, CurlSpeed, OveralSpeedLoop, Param4]}.OveralSpeedLoop
   )
  )
  n34 = Frac(None=Add(A=Frac(None=@n33), B=PerInstanceRandom()))
  n35 = Multiply(
   A=Multiply(A=LinearInterpolate(A=@n32, B=@n31, Alpha=@n34), B=ScalarParameter() {defaultValue=1, group=WPO, parameterName=OverallSpeed})
   B=LinearInterpolate(
    A=VectorParameter() {parameterName=OverallDirection, defaultValue=<rgba 1 0 0 1>}.RGB
    B=VectorParameter() {parameterName=OverallDirectionMax, defaultValue=<rgba 1 0 0 1>}.RGB
    Alpha=PerInstanceRandom()
   )
  )
  n36 = Add(
   A=MaterialFunctionCall(WidthBase=@n13, WidthTip=Multiply(A=@n14, B=@n12)) {materialFunction=&SplineThicken}.WorldPositionOffset
   B=AppendVector(A=AppendVector(A=@n29, B=@n30), B=ComponentMask(None=@n35) {b=true})
  )
  n37 = MaterialFunctionCall(Offset Amount=ScalarParameter() {parameterName=CameraOffset}) {materialFunction=/Engine/Functions/Engine_MaterialFunctions02/WorldPositionOffset/CameraOffset}
  NamedRerouteDeclaration(None=StaticSwitchParameter(True=Add(A=@n36, B=@n37.Clamped Camera Offset), False=@n36) {parameterName=UseCameraOffset?}) {name=WPO, nodeColor=<rgba 1 0.8438 0 1>}
  n38 = MaterialFunctionCall() {materialFunction=&LinearGradient}
  n39 = OneMinus(None=@n38.UGradient)
  n40 = OneMinus(None=@n38.VGradient)
  n41 = Multiply(
   A=Multiply(A=Multiply(A=@n39, B=@n38.UGradient), B=Multiply(A=@n38.VGradient, B=@n40))
   B=ScalarParameter() {defaultValue=2, parameterName=GradientAmount}
  )
  n42 = LinearInterpolate(
   A=ScalarParameter() {defaultValue=1, parameterName=TileX_Main}
   B=ScalarParameter() {defaultValue=1, parameterName=TileX2_Main}
   Alpha=PerInstanceRandom()
  ) {constA=0.1, constB=0.25}
  n43 = TextureCoordinate() {coordinateIndex=1}
  n44 = ComponentMask(None=@n43) {r=true}
  n45 = ScalarParameter() {defaultValue=1, parameterName=TileY_Main}
  n46 = ComponentMask(None=@n43) {g=true}
  n47 = ScalarParameter() {defaultValue=0.15, parameterName=SpeedX}
  n48 = ScalarParameter() {defaultValue=0.15, parameterName=SpeedX2}
  n49 = PerInstanceRandom()
  n50 = ScalarParameter() {defaultValue=0.15, parameterName=SpeedY}
  n51 = ScalarParameter() {defaultValue=0.15, parameterName=SpeedY2}
  n52 = PerInstanceRandom()
  n53 = Panner(
   Coordinate=AppendVector(A=Multiply(A=@n42, B=@n44), B=Multiply(A=@n45, B=@n46))
   Speed=AppendVector(
    A=LinearInterpolate(A=@n47, B=@n48, Alpha=@n49) {constA=0.1, constB=0.25}
    B=LinearInterpolate(A=@n50, B=@n51, Alpha=@n52) {constA=0.1, constB=0.25}
   )
  ) {speedX=0.15}
  n54 = TextureCoordinate()
  n55 = LinearInterpolate(Alpha=PerInstanceRandom()) {constA=0.5}
  n56 = AppendVector(
   A=Multiply(A=@n55, B=ScalarParameter() {defaultValue=1, group=Texture values, parameterName=TileX})
   B=ScalarParameter() {defaultValue=1, group=Texture values, parameterName=TileY}
  )
  n57 = ScalarParameter() {defaultValue=0.15, group=Texture values, parameterName=SpeedY}
  n58 = Multiply(A=@n55, B=ScalarParameter() {defaultValue=0.1, group=Texture values, parameterName=SpeedMult})
  n59 = TextureSampleParameter2D(UVs=Multiply(A=TextureCoordinate(), B=@n56)) {
   parameterName=DistortionTexture
   samplerType=SAMPLERTYPE_LinearColor
   texture=/Game/Global/Textures/Default/Tileable_Patterns/T_Tile_Smoke_Small
  }
  n60 = TextureSampleParameter2D(UVs=Multiply(A=TextureCoordinate(), B=@n56)) {
   parameterName=DistortionTexture
   samplerType=SAMPLERTYPE_LinearColor
   texture=&T_Tile_Smoke_Small
  }
  n61 = Multiply(
   A=Add(A=ComponentMask(None=@n59.RGB) {g=true, r=true}, B=ComponentMask(None=@n60.RGB) {g=true, r=true})
   B=ScalarParameter() {defaultValue=0.1, group=Texture values, parameterName=BaseDistort}
  )
  n62 = Constant()
  n63 = ShadingPathSwitch(Default=QualitySwitch(Default=ComponentMask(None=@n61) {g=true, r=true}, Low=@n62), Mobile=@n62) {desc=Migrated from FeatureLevelSwitch}
  n64 = Add(A=Panner(Coordinate=Multiply(A=@n54, B=@n56), Speed=AppendVector(A=@n57, B=@n58)) {speedY=1}, B=@n63)
  n65 = ScalarParameter() {defaultValue=0.15, group=Texture values, parameterName=SpeedY}
  n66 = Multiply(A=@n55, B=ScalarParameter() {defaultValue=0.1, group=Texture values, parameterName=SpeedMult_02})
  n67 = Add(A=Panner(Coordinate=Multiply(A=@n54, B=@n56), Speed=AppendVector(A=@n65, B=@n66)) {speedY=1}, B=@n63)
  n68 = Add(
   A=TextureSampleParameter2D(UVs=VertexInterpolator(VS=@n64).PS) {parameterName=SmokeTexture, texture=/Game/Global/Textures/Default/Caustics/T_Turbulence_Seamless}.G
   B=TextureSampleParameter2D(UVs=VertexInterpolator(VS=@n67).PS) {parameterName=SmokeTexture, texture=&T_Turbulence_Seamless}.G
  )
  n69 = ScalarParameter() {defaultValue=0.15, parameterName=Contrast}
  n70 = Multiply(
   A=TextureSampleParameter2D(UVs=VertexInterpolator(VS=@n53).PS) {parameterName=MainTexture, texture=/Game/Effects/Fort_Effects/Textures/Smoke/T_Puffy_Smoke_Trail_01}.G
   B=LinearInterpolate(A=Constant() {r=1}, B=Power(Base=@n68, Exp=@n69), Alpha=ScalarParameter() {defaultValue=1, parameterName=ExtraTextureAmount})
  )
  n71 = ScalarParameter() {defaultValue=1, parameterName=Opacity}
  n72 = ScalarParameter() {defaultValue=1, parameterName=OpacityFar}
  n73 = Multiply(A=Multiply(A=Multiply(A=@n41, B=@n70), B=LinearInterpolate(A=@n71, B=@n72, Alpha=@n11)), B=PerInstanceFadeAmount())
  n74 = ParticleColor()
  NamedRerouteDeclaration(None=DepthFade(Opacity=Multiply(A=@n73, B=@n74.A), FadeDistance=ScalarParameter() {defaultValue=25, parameterName=DepthFadeDistance})) {name=Opacity, nodeColor=<rgba 1 0 0.1172 1>}
  NamedRerouteUsage()
  NamedRerouteUsage()
  StaticSwitchParameter(True=Divide(A=Time() {bOverride_Period=true, period=6}) {constB=6}, False=Divide(A=Time() {bOverride_Period=true, period=12}) {constB=12}) {defaultValue=true, parameterName=UseFastSprites}
 ]
}
